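# Niagara effect for the DandyFlare emote's lights, chiefly a persistent purple thruster drawn with a mesh and a sprite.
# Thruster burst-spawns particles that outlive their lifetime.
# It compares the owner's summed absolute velocity components to a threshold, smooths the result, and uses it to brighten the sprite and lengthen the mesh in Z while moving.
# RedLight has no renderers, and YellowLights is disabled.
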NiagaraSystem /BRCosmetics/Animation/Game/MainPlayer/Emotes/DandyFlare/FX/NS_Emote_DandyFlare_Lights
dependencies {
 usedDataInterfaces []
 usedDynamicInputs [
  &FN_FrontEndFloatModifier
  &MakeLinearColorFromVectorAndFloat
  &FN_FrontEndVector3Modifier
  &MakeVectorFromLinearColorRGB
 ]
 usedModules [
  &self:SystemSpawnScript.SystemScriptSource.SystemScriptGraph.NiagaraNodeAssignment_0.SetVariables_#1
  &SystemState
 ]
}
emitters {
 RedLight {rendererClasses [], renderers []}
 Thruster {
  emitterUpdateScript [
   &self:InterpolateSpeedThresholdEmitter_02(
    Define Initial Value=false
    Rate Of Change=5
    Target Float Value:NiagaraFloat=&MakeCustomFloatFromBool(
     Bool=&SetBoolByFloatComparison(
      A=&Add_Float(
       A=&Add_Float(
        A=&Abs_Float(Float:NiagaraFloat=&MakeFloatFromVector(Vector=$Engine.Owner.Velocity, Channel=::NewEnumerator0))
        B=&Abs_Float(Float:NiagaraFloat=&MakeFloatFromVector(Vector=$Engine.Owner.Velocity, Channel=::NewEnumerator2))
       )
       B=&Abs_Float(Float:NiagaraFloat=&MakeFloatFromVector(Vector=$Engine.Owner.Velocity, Channel=::NewEnumerator1))
      )
      B=33.6457
     )
     True Float=1
     False Float=0
    )
   ) {module=InterpolateSpeedThresholdEmitter_02}
   &EmitterState(
    Inactive Response=::NewEnumerator1
    Life Cycle Mode=::NewEnumerator1
    Loop Behavior=::NewEnumerator1
    Loop Duration Mode=::NewEnumerator1
    UseLoopDelay=false
   )
   &SpawnBurst_Instantaneous(Use Spawn Probability=false, UseLoopCountLimit=false)
  ]
  particleSpawnScript [
   &InitializeParticle(
    Color=<rgba 1.2434 0.4874 10 1>
    Lifetime=0.8
    Mesh Scale=<0.13 0.002 4>
    Mesh Scale Mode=::NewEnumerator4
    Position Offset=<0 -34.3388 4.5165>
    Sprite Size=<11.1153 56.2761>
    Sprite Size Mode=::NewEnumerator4
    UsePositionOffset=true
    Write Material Random=true
   )
   &InitialMeshOrientation(Mesh Orientation Mode=::NewEnumerator7, Rotation=<0.25 0 0>, Use Rotation Vector=true)
   &DynamicMaterialParameters(
    Index 0 Param 4=1.5
    Module.Index0Float0Write=true
    Module.Index0Float1Write=true
    Module.Index0Float2Write=true
    Module.Index0Float3Write=true
   )
   &SpriteFacingAndAlignment(Sprite Alignment=<1 0 0>, WriteAlignment=true, WriteFacing=false)
  ]
  particleUpdateScript [
   /SetVariables(
    Particles.SpriteColor=&LerpLinearColors(
     StartColor=<rgba 1.3879 0.183 2.087 0.0348>
     EndColor=<rgba 3.2546 0.3201 6 0.0783>
     LerpFactor=$Emitter.InterpolateSpeedThresholdEmitter_02.InterpolateOverTime_Float.Moving Average
    )
    Particles.SpritePos:NiagaraPosition=&ConvertPositionToVector(Input Position=<0 -47.1713 4.5165>, Conversion Operation:ENiagara_LWCConvertPosToVec=::NewEnumerator0)
   ) {isSetParameters=true}
   &ParticleState(Kill Particles When Lifetime Has Elapsed=false)
   &ScaleMeshSize(Scale Factor=&MakeVector(
     X=1
     Y=1
     Z=&Lerp_Float(A=0.3, B=0.7, Alpha=$Emitter.InterpolateSpeedThresholdEmitter_02.InterpolateOverTime_Float.Moving Average)
    ))
   &SolveForcesAndVelocity(Clamp Velocity=false, Limit Acceleration=false)
  ]
  rendererClasses [&NiagaraMeshRendererProperties, &NiagaraSpriteRendererProperties]
  renderers [
   {rendererClass=&NiagaraMeshRendererProperties, rendererIndex=0}
   {rendererClass=&NiagaraSpriteRendererProperties, rendererIndex=1}
  ]
 }
 YellowLights {bEnabled=false, rendererClasses [], renderers []}
}
summary {
 systemName=NS_Emote_DandyFlare_Lights
 emitters [
  {bEnabled=false, emitterName=YellowLights, rendererClasses []}
  {emitterName=RedLight, rendererClasses []}
  {emitterName=Thruster, rendererClasses [&NiagaraMeshRendererProperties, &NiagaraSpriteRendererProperties]}
 ]
 userVariables [
  User.bIsFrontEnd {type=NiagaraBool}
  User.bIsFrontEndPreview {type=NiagaraBool}
 ]
}
