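# Niagara effect for the DandyFlare emote that renders gun-shot flashes and glowing barrel balls on the gadget's skeletal mesh.
# Projectiles spawns two particle groups when User.BurstIndex changes to a nonzero value, gated by system age, pinning short-lived mesh, sprite and light visuals to the pick-gun socket.
# Curves shape flash scale and glow, with separate front-end and gameplay glow colors.
# RotatingBarrelBalls keeps two mesh groups attached across three spinner sockets via partitioning.
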NiagaraSystem /BRCosmetics/Animation/Game/MainPlayer/Emotes/DandyFlare/FX/NS_Emote_DandyFlare_Projectiles
dependencies {
 usedDataInterfaces [
  &NiagaraDataInterfaceCurve
  &NiagaraDataInterfaceVectorCurve
  &NiagaraDataInterfaceSkeletalMesh
  &NiagaraDataInterfaceColorCurve
  &NiagaraDataInterfaceArrayInt32
 ]
 usedDynamicInputs []
 usedModules [&SystemState]
}
emitters {
 Projectiles {
  emitterUpdateScript [
   &EmitterState(Life Cycle Mode=::NewEnumerator1, Loop Duration=4.5319, UseLoopDelay=false)
   /Game/Effects/Niagara/Modules/Logic/UpdateValueStateChange(InputIntParam:NiagaraInt32=&MakeIntFromFloat(Float=$User.BurstIndex, Conversion Mode=::NewEnumerator1))
   &self:TriggerSpawn(
    Spawn Count:NiagaraInt32=&MergeHelper_IntSwitch(
     0=0
     1=1
     2=1
     3=1
     4=1
     5=1
     6=1
     7=1
     Index=&MakeIntFromFloat(Float=$User.BurstIndex, Conversion Mode=::NewEnumerator1)
     Num=::NewEnumerator6
    )
    Spawn Group:NiagaraInt32=&MakeIntFromFloat(Float=0, Conversion Mode=::NewEnumerator1)
    Trigger:NiagaraBool=&Boolean_AndOperation(Bool1=$Emitter.UpdateValueStateChange.Changed, Bool2=&SetBoolByFloatComparison(A=$System.Age, B=0))
   ) {module=TriggerSpawn}
   &self:TriggerSpawn(
    Spawn Count:NiagaraInt32=&MergeHelper_IntSwitch(
     0=0
     1=1
     2=1
     3=1
     4=1
     5=1
     6=1
     7=1
     Index=&MakeIntFromFloat(Float=$User.BurstIndex, Conversion Mode=::NewEnumerator1)
     Num=::NewEnumerator6
    )
    Spawn Group:NiagaraInt32=&MakeIntFromFloat(Float=1, Conversion Mode=::NewEnumerator1)
    Trigger:NiagaraBool=&Boolean_AndOperation(Bool1=$Emitter.UpdateValueStateChange.Changed, Bool2=&SetBoolByFloatComparison(A=$System.Age, B=0.04))
   ) {module=TriggerSpawn001}
  ]
  particleSpawnScript [
   /SetVariables(Particles.VisibilityTag=&SpawnGroup()) {isSetParameters=true}
   &InitializeParticle(
    Color=&MergeHelper_ColorSwitch(
     0=<rgba 0.4806 1 0 1>
     1=<rgba 2.4598 3 0 1>
     2=<rgba 1 0.7271 0 1>
     3=<rgba 3 1.2863 0 0.287>
     Index=&SpawnGroup()
     Num=::NewEnumerator2
    )
    Lifetime=&MergeHelper_FloatSwitch(0=0.116, 1=0.12, 2=4.4004, 3=0.106, Index=&SpawnGroup(), Num=::NewEnumerator2)
    Mass Max=1
    Mass Min=1
    Mass Mode=::NewEnumerator3
    Mesh Scale Mode=::NewEnumerator3
    Mesh Uniform Scale=&MergeHelper_FloatSwitch(0=-0.216, 1=0.17, 2=3.9689, 3=0, 4=0, Index=&SpawnGroup(), Num=::NewEnumerator3)
    Ribbon Width Mode=::NewEnumerator1
    Sprite Size Mode=::NewEnumerator3
    Uniform Sprite Size=&MergeHelper_FloatSwitch(0=0.2668, 1=0.0588, 2=0.7633, 3=?, Index=&SpawnGroup(), Num=::NewEnumerator1)
    UsePositionOffset=false
    Write Material Random=true
   )
   /SetVariables(Particles.SpriteCameraOffset=10) {isSetParameters=true}
  ]
  particleUpdateScript [
   &ParticleState()
   &ScaleSpriteSize(
    Scale Sprite Size Mode=::NewEnumerator0
    Uniform Scale Factor=&MergeHelper_FloatSwitch(0=180, 1=0, 2=?, 3=?, Index=0, Num=::NewEnumerator0)
   )
   &DynamicMaterialParameters(
    Index 0 Param 1=&FloatFromCurve(FloatCurve=DataInterface() {propertyValues {Curve {defaultValue=1, keys [{tangentMode=RCTM_User, time=0, value=0}, {tangentMode=RCTM_User, time=1, value=1}]}}})
    Index 0 Param 2=0.448
    Module.Index0Float0Write=true
    Module.Index0Float1Write=true
    Module.Index0Float2Write=true
    Module.Index0Float3Write=true
   )
   &ScaleMeshSize(Scale Factor=&VectorFromCurve(VectorCurve=DataInterface() {propertyValues {
       XCurve {
        defaultValue=0
        keys [{interpMode=RCIM_Cubic, tangentMode=RCTM_User, time=0.7, value=1}, {interpMode=RCIM_Cubic, tangentMode=RCTM_User, time=1, value=0}]
       }
       YCurve {
        defaultValue=0
        keys [{interpMode=RCIM_Cubic, tangentMode=RCTM_User, time=0.7, value=1}, {interpMode=RCIM_Cubic, tangentMode=RCTM_User, time=1, value=0}]
       }
       ZCurve {
        defaultValue=0
        keys [{interpMode=RCIM_Cubic, tangentMode=RCTM_User, time=0.7, value=1}, {interpMode=RCIM_Cubic, tangentMode=RCTM_User, time=1, value=0}]
       }
      }}))
   &SolveForcesAndVelocity(Clamp Velocity=false, Limit Acceleration=false)
   &SkeletalMeshLocation(
    MaskSpawnGroup=false
    Mesh Sampling Type=::NewEnumerator5
    Sampled Position Offset=<5.5 0 0>
    Skeletal Mesh=DataInterface() {propertyValues {
      DefaultMesh=null
      ExcludeBoneName=root
      OptionalComponentClass=null
      PreviewMesh=/BRCosmetics/Animation/Props/Gadgets_DandyFlare/Meshes/DandyFlare_Gadget
      SkinningMode=SkinOnTheFly
      SoftSourceActor=null
      SourceMode=Default
      UvSetIndex=0
      WholeMeshLOD=-1
      bExcludeBone=false
      bReadDeformedGeometry=true
      bRequireCurrentFrameData=true
      ComponentTags []
      FilteredBones []
      FilteredSockets [JNT_PickGunSocket]
      MeshUserParameter {parameter {name=None, typeDefHandle {registeredTypeIndex=1}, varData []}}
      SamplingRegions []
     }}
    Socket Sampling Mode=::NewEnumerator5
   )
   /SetVariables(Particles.MuzzleFlashScale:Vector3f=&VectorFromCurve(
     VectorCurve=DataInterface() {propertyValues {
       XCurve {
        defaultValue=0
        keys [
         {interpMode=RCIM_Cubic, tangentMode=RCTM_User, time=0, value=0}
         {arriveTangent=-1.7309, interpMode=RCIM_Cubic, leaveTangent=-1.7309, tangentMode=RCTM_User, time=0.0344, value=0.9705}
        ]
       }
       YCurve {
        defaultValue=1
        keys [
         {interpMode=RCIM_Cubic, tangentMode=RCTM_User, time=0, value=0}
         {arriveTangent=-0.4055, interpMode=RCIM_Cubic, leaveTangent=-0.4055, tangentMode=RCTM_User, time=0.0344, value=0.9705}
         {arriveTangent=-0.4857, interpMode=RCIM_Cubic, leaveTangent=-0.4857, tangentMode=RCTM_User, time=0.2998, value=0.891}
         {tangentMode=RCTM_User, time=1, value=0}
        ]
       }
       ZCurve {
        defaultValue=0
        keys [
         {interpMode=RCIM_Cubic, tangentMode=RCTM_User, time=0, value=0}
         {arriveTangent=-0.367, interpMode=RCIM_Cubic, leaveTangent=-0.367, tangentMode=RCTM_User, time=0.0344, value=0.9705}
         {arriveTangent=-0.4857, interpMode=RCIM_Cubic, leaveTangent=-0.4857, tangentMode=RCTM_User, time=0.2998, value=0.9107}
         {tangentMode=RCTM_User, time=1, value=0}
        ]
       }
      }}
     Scale Curve=<0.5 0.5 0.5>
    )) {isSetParameters=true}
   &FN_Light_Attributes(
    Color=&ColorFromCurve(ColorCurve=DataInterface() {propertyValues {
       AlphaCurve {keys [<at 0.1054 1 RCIM_Cubic>,<at 0.1376 0 RCIM_Cubic>]}
       BlueCurve {keys [<at 0 0 RCIM_Cubic>]}
       GreenCurve {keys [<at 0 10 RCIM_Cubic>]}
       RedCurve {keys [<at 0 4.8659 RCIM_Cubic>]}
      }})
    Create Color Attribute=true
    Create Radius Attribute=true
    Radius=67.1444
   ) {module=FN_Light_Attributes001}
   /SetVariables(
    Particles.GlowPos:NiagaraPosition=&AddVectorToPosition(Position=$Particles.Position, Vector=<0 0 0>)
    Particles.MuzzleGlowColor:LinearColor=&MakeLinearColorFromVectorAndFloat(
     "Vector (RGB)"=&FN_FrontEndVector3Modifier(
      Gameplay=&MakeVectorFromLinearColorRGB(Color=&ColorFromCurve(ColorCurve=DataInterface() {propertyValues {
          AlphaCurve {keys [<at 0 1 RCIM_Cubic>]}
          BlueCurve {keys [<at 0 0 RCIM_Cubic>,<at 0.0511 0>,<at 0.9889 0>]}
          GreenCurve {keys [<at 0 20 RCIM_Cubic>,<at 0.0511 10>,<at 0.9889 6>]}
          RedCurve {keys [<at 0 16.3987 RCIM_Cubic>,<at 0.0511 8.1994>,<at 0.9889 4.9196>]}
         }}))
      Front End=&MakeVectorFromLinearColorRGB(Color=&ColorFromCurve(ColorCurve=DataInterface() {propertyValues {
          AlphaCurve {keys [<at 0 1 RCIM_Cubic>]}
          BlueCurve {keys [<at 0 0 RCIM_Cubic>,<at 0.0511 0>,<at 0.9889 0>]}
          GreenCurve {keys [<at 0 10 RCIM_Cubic>,<at 0.0511 3>,<at 0.9889 1>]}
          RedCurve {keys [<at 0 8.1994 RCIM_Cubic>,<at 0.0511 2.4598>,<at 0.9889 0.8199>]}
         }}))
     )
     "Float (Alpha)"=&FloatFromCurve(FloatCurve=DataInterface() {propertyValues {Curve {keys [<at 0 1>,<at 1 0>]}}})
    )
   ) {isSetParameters=true}
  ]
  rendererClasses [
   &NiagaraMeshRendererProperties
   &NiagaraLightRendererProperties
   &NiagaraSpriteRendererProperties
  ]
  renderers [
   {rendererClass=&NiagaraMeshRendererProperties, rendererIndex=0}
   {rendererClass=&NiagaraLightRendererProperties, rendererIndex=1}
   {rendererClass=&NiagaraSpriteRendererProperties, rendererIndex=2}
  ]
 }
 RotatingBarrelBalls {
  emitterUpdateScript [
   &EmitterState(
    Inactive Response=::NewEnumerator1
    Life Cycle Mode=::NewEnumerator1
    Loop Behavior=::NewEnumerator1
    Loop Duration Mode=::NewEnumerator1
    UseLoopDelay=false
   )
   &SpawnBurst_Instantaneous(Spawn Count=3, Use Spawn Probability=false, UseLoopCountLimit=false)
   &SpawnBurst_Instantaneous(Spawn Count=3, Spawn Group=1, Use Spawn Probability=false, UseLoopCountLimit=false) {module=SpawnBurst_Instantaneous001}
  ]
  particleSpawnScript [
   /SetVariables(Particles.VisibilityTag=&SpawnGroup()) {isSetParameters=true}
   &InitializeParticle(
    Color=&MergeHelper_ColorSwitch(0=<rgba 2.4598 3 0 1>, 1=<rgba 2.4598 3 0 0.2609>, 2=?, 3=?, Index=&SpawnGroup(), Num=::NewEnumerator0)
    Mass Max=1
    Mass Min=1
    Mass Mode=::NewEnumerator3
    Mesh Scale Mode=::NewEnumerator3
    Mesh Uniform Scale=&MergeHelper_FloatSwitch(0=0.15, 1=0.13, 2=0, 3=0, 4=0, Index=&SpawnGroup(), Num=::NewEnumerator3)
    Ribbon Width Mode=::NewEnumerator1
    Sprite Size Mode=::NewEnumerator3
    Uniform Sprite Size=0.5556
    UsePositionOffset=false
    Write Material Random=true
   )
   &PartitionParticles(Number Of Partitions=3)
  ]
  particleUpdateScript [
   &ParticleState(Kill Particles When Lifetime Has Elapsed=false)
   &SkeletalMeshLocation(
    Bone / Socket Index=&SelectIntFromArray(Direct Array Index=$Particles.PartitionParticles.Partition, Int Selection Array=DataInterface() {propertyValues {IntData [0,1,2]}})
    MaskSpawnGroup=false
    Mesh Sampling Type=::NewEnumerator5
    Skeletal Mesh=DataInterface() {propertyValues {
      DefaultMesh=null
      ExcludeBoneName=root
      OptionalComponentClass=null
      PreviewMesh=&DandyFlare_Gadget
      SkinningMode=SkinOnTheFly
      SoftSourceActor=null
      SourceMode=Default
      UvSetIndex=0
      WholeMeshLOD=-1
      bExcludeBone=false
      bReadDeformedGeometry=true
      bRequireCurrentFrameData=true
      ComponentTags []
      FilteredBones []
      FilteredSockets [JNT_SpinnerSocket, JNT_SpinnerSocket_0, JNT_SpinnerSocket_1]
      MeshUserParameter {parameter {name=None, typeDefHandle {registeredTypeIndex=1}, varData []}}
      SamplingRegions []
     }}
    Socket Sampling Mode=::NewEnumerator5
   )
   &DynamicMaterialParameters(
    Index 0 Param 1=&FloatFromCurve(FloatCurve=DataInterface() {propertyValues {Curve {defaultValue=1, keys [{tangentMode=RCTM_User, time=1, value=1}]}}})
    Index 0 Param 2=0.448
    Module.Index0Float0Write=true
    Module.Index0Float1Write=true
    Module.Index0Float2Write=true
    Module.Index0Float3Write=true
   )
   &SolveForcesAndVelocity(Clamp Velocity=false, Limit Acceleration=false)
  ]
  rendererClasses [&NiagaraMeshRendererProperties]
  renderers [
   {rendererClass=&NiagaraMeshRendererProperties, rendererIndex=0}
   {rendererClass=&NiagaraMeshRendererProperties, rendererIndex=1}
  ]
 }
}
summary {
 systemName=NS_Emote_DandyFlare_Projectiles
 emitters [
  {
   emitterName=Projectiles
   rendererClasses [
    &NiagaraMeshRendererProperties
    &NiagaraLightRendererProperties
    &NiagaraSpriteRendererProperties
   ]
  }
  {emitterName=RotatingBarrelBalls, rendererClasses [&NiagaraMeshRendererProperties]}
 ]
 userVariables [
  User.bIsFrontEnd {type=NiagaraBool}
  User.bIsFrontEndPreview {type=NiagaraBool}
  User.BurstIndex {type=NiagaraFloat, defaultValue=1, description=""}
  User.OwnerLocalVelocity {type=Vector3f, description="", defaultValue {struct=Vector3f, value=<0 0 0>}}
 ]
}
